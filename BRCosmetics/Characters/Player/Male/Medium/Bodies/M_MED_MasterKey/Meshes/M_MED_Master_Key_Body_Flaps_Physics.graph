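# Physics asset for the coat flaps on the medium male Master Key player body, defining physics shapes for three coat-tail bones and nearby body parts.
# Box bodies cover the back, front-left and front-right tails, each constrained to the pelvis with narrow swing and twist limits, with the back tail tightest.
# Simple capsules and a sphere on the pelvis, spine and thighs approximate the body volume around the flaps.
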PhysicsAsset /BRCosmetics/Characters/Player/Male/Medium/Bodies/M_MED_MasterKey/Meshes/M_MED_Master_Key_Body_Flaps_Physics
bodies {
 dyn_coat_tail_back [{shapeType=Box, extentX=14.7296, extentY=21.1977, extentZ=3.2265, center=<7.7125 0 0>}]
 dyn_coat_tail_front_l [{shapeType=Box, extentX=19.8885, extentY=14.8135, extentZ=3.5674, center=<10.0939 0 0>}]
 dyn_coat_tail_front_r [{shapeType=Box, extentX=19.9649, extentY=13.682, extentZ=2.7902, center=<10.344 0 -0.0561>}]
 pelvis [{shapeType=Capsule, length=0, radius=12}]
 spine_02 [{shapeType=Sphere, radius=15.5}]
 spine_05 [{shapeType=Capsule, length=12.0134, radius=13.3957, center=<-7.1937 0 0>}]
 thigh_l [
  {shapeType=Capsule, length=15.5, radius=7.5, center=<-16.8299 1.7951 -7.2915>, rotation=<rot 54.9802 -174.3925 0.7611>}
  {shapeType=Capsule, length=11.135, radius=8.1607, center=<-16.2778 3.0418 0.0005>, rotation=<rot 90 0 -166.9654>}
 ]
 thigh_r [{shapeType=Capsule, length=18.0974, radius=7.457, center=<33.1103 -3.3346 2.989>, rotation=<rot 90 0 0>}]
}
constraints [
 {bone1Name=dyn_coat_tail_back, bone2Name=pelvis, swing1LimitDegrees=5, swing2LimitDegrees=15, twistLimitDegrees=3}
 {bone1Name=dyn_coat_tail_front_l, bone2Name=pelvis, swing1LimitDegrees=15, swing2LimitDegrees=15, twistLimitDegrees=5}
 {bone1Name=dyn_coat_tail_front_r, bone2Name=pelvis, swing1LimitDegrees=15, swing2LimitDegrees=15, twistLimitDegrees=5}
]
