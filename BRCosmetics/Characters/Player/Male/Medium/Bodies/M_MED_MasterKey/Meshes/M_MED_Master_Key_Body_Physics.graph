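# Physics asset for the medium male Master Key player body, set up for two clusters of dangling keys on the left and right.
# Each dyn_key_mstr box is constrained to the pelvis, with much looser limits on the left, and parents capsule keys (four left, three right) through limited swing and twist.
# Simple shapes cover pelvis, spine, arms and thighs; among these, only spine_05 and pelvis share a constraint, with free rotation.
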PhysicsAsset /BRCosmetics/Characters/Player/Male/Medium/Bodies/M_MED_MasterKey/Meshes/M_MED_Master_Key_Body_Physics
bodies {
 const_key_mstr_l [{shapeType=Box, extentX=1.591, extentY=1.591, extentZ=1.591}]
 const_key_mstr_r [{shapeType=Box, extentX=1.2274, extentY=1.2274, extentZ=1.2274}]
 dyn_key_l_back [{shapeType=Capsule, length=9.1709, radius=2, center=<6.2265 0 0>, rotation=<rot 90 0 0>}]
 dyn_key_l_front [{shapeType=Capsule, length=6.934, radius=1.5, center=<4.4912 -0.0001 0>, rotation=<rot 90 0 0>}]
 dyn_key_l_mid_back [{shapeType=Capsule, length=10.031, radius=2.5, center=<7.1281 0.3941 0.8191>, rotation=<rot 90 0 0>}]
 dyn_key_l_mid_front [{shapeType=Capsule, length=9.4404, radius=1.5, center=<5.954 0 0>, rotation=<rot 90 0 0>}]
 dyn_key_mstr_l [{shapeType=Box, extentX=16.25, extentY=15, extentZ=5.75, center=<12.1824 -1.1499 0.4121>, rotation=<rot -5.7265 -8.7159 -4.0088>}]
 dyn_key_mstr_r [{shapeType=Box, extentX=16.25, extentY=15, extentZ=5.75, center=<9.2583 -0.0139 0>}]
 dyn_key_r_back [{shapeType=Capsule, length=14.8132, radius=1.7994, center=<8.9939 0 0.0001>, rotation=<rot 90 0 0>}]
 dyn_key_r_front [{shapeType=Capsule, length=10.2675, radius=2.5, center=<7.3978 0 0>, rotation=<rot 90 0 0>}]
 dyn_key_r_mid [{shapeType=Capsule, length=11.781, radius=2.5, center=<8.3694 0 0>, rotation=<rot 90 0 0>}]
 lowerarm_l [{shapeType=Capsule, length=23.125, radius=6.875, center=<17.9383 -0.0002 0>, rotation=<rot 90 -0.0132 -0.0131>}]
 lowerarm_r [{shapeType=Capsule, length=23.875, radius=6.125, center=<-14.3142 0 0>, rotation=<rot -90 0.0132 -0.0132>}]
 pelvis [{shapeType=Capsule, length=0, radius=12}]
 spine_02 [{shapeType=Sphere, radius=15.5}]
 spine_05 [{shapeType=Capsule, length=12.0134, radius=13.3957, center=<-7.1937 0 0>}]
 thigh_l [
  {shapeType=Capsule, length=16.375, radius=10.5, center=<-12.2589 2.1432 -1.4595>, rotation=<rot 70.1923 -173.8147 1.2801>}
  {shapeType=Capsule, length=24.25, radius=8.25, center=<-15 0 0>, rotation=<rot 90 0 180>}
 ]
 thigh_r [
  {shapeType=Capsule, length=23.25, radius=7.25, center=<15 0 0>, rotation=<rot 90 0 0>}
  {shapeType=Capsule, length=7.5002, radius=10.5, center=<28.7646 -0.0002 0.1355>, rotation=<rot -84.9999 0.0001 0>}
 ]
 upperarm_twist_01_l [{shapeType=Capsule, length=23.875, radius=8, center=<14.4503 0.0001 0>, rotation=<rot 90 0.0395 0.0394>}]
 upperarm_twist_02_r [{shapeType=Capsule, length=23.25, radius=6.75, rotation=<rot -85.9641 -179.9999 179.9999>}]
}
constraints [
 {bone1Name=dyn_key_mstr_r, bone2Name=pelvis, swing1LimitDegrees=40, swing2LimitDegrees=40, twistLimitDegrees=5}
 {bone1Name=dyn_key_mstr_l, bone2Name=pelvis, swing1LimitDegrees=90, swing2LimitDegrees=90, twistLimitDegrees=45}
 {
  bone1Name=spine_05
  bone2Name=pelvis
  swing1LimitDegrees=45
  swing1Motion=Free
  swing2LimitDegrees=45
  swing2Motion=Free
  twistLimitDegrees=45
  twistMotion=Free
 }
 {bone1Name=dyn_key_l_front, bone2Name=dyn_key_mstr_l, swing1LimitDegrees=40, swing2LimitDegrees=40, twistLimitDegrees=10}
 {bone1Name=dyn_key_l_back, bone2Name=dyn_key_mstr_l, swing1LimitDegrees=40, swing2LimitDegrees=40, twistLimitDegrees=10}
 {bone1Name=dyn_key_l_mid_back, bone2Name=dyn_key_mstr_l, swing1LimitDegrees=40, swing2LimitDegrees=40, twistLimitDegrees=10}
 {bone1Name=dyn_key_l_mid_front, bone2Name=dyn_key_mstr_l, swing1LimitDegrees=40, swing2LimitDegrees=40, twistLimitDegrees=10}
 {bone1Name=dyn_key_r_back, bone2Name=dyn_key_mstr_r, swing1LimitDegrees=40, swing2LimitDegrees=40, twistLimitDegrees=10}
 {bone1Name=dyn_key_r_front, bone2Name=dyn_key_mstr_r, swing1LimitDegrees=40, swing2LimitDegrees=40, twistLimitDegrees=10}
 {bone1Name=dyn_key_r_mid, bone2Name=dyn_key_mstr_r, swing1LimitDegrees=40, swing2LimitDegrees=40, twistLimitDegrees=10}
]
