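# Physics asset for the male Master Key character's neck piece, simulating a dangling set of keys against simplified body shapes.
# A box on the master neck bone hangs from spine_05 with tight angular limits, and left, right and center key capsules attach to it with wider limits.
# Spine, pelvis and arm shapes approximate the body, with spine_05 joined to the pelvis by free swing and twist.
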PhysicsAsset /BRCosmetics/Characters/Player/Male/Medium/Bodies/M_MED_MasterKey/Meshes/M_MED_Master_Key_Neck_Physics
bodies {
 dyn_key_mstr_neck [{shapeType=Box, extentX=12, extentY=11, extentZ=2.75, center=<9.0678 0 0.4952>, rotation=<rot -10 0 0>}]
 dyn_key_neck_c [{shapeType=Capsule, length=7.809, radius=1.4, center=<7.3699 0.0044 0.4326>, rotation=<rot 0 90.0375 -90.274>}]
 dyn_key_neck_l [{shapeType=Capsule, length=2.3976, radius=2.125, center=<4.338 0.0255 -1.0995>, rotation=<rot 0 -89.5813 -50.8546>}]
 dyn_key_neck_r [{shapeType=Capsule, length=6.0863, radius=1.2, center=<6.2621 -0.0346 0.3464>, rotation=<rot 0 89.8737 -90.3167>}]
 lowerarm_l [{shapeType=Capsule, length=23.125, radius=6.875, center=<17.9383 -0.0002 0>, rotation=<rot 90 -0.0132 -0.0131>}]
 lowerarm_r [{shapeType=Capsule, length=23.875, radius=6.125, center=<-14.3142 0 0>, rotation=<rot -90 0.0132 -0.0132>}]
 pelvis [{shapeType=Capsule, length=0, radius=12}]
 spine_02 [{shapeType=Sphere, radius=15.5}]
 spine_05 [{shapeType=Capsule, length=12.0134, radius=13.3957, center=<-7.1937 0 0>}]
 upperarm_twist_01_l [{shapeType=Capsule, length=23.875, radius=8, center=<14.4503 0.0001 0>, rotation=<rot 90 0.0395 0.0394>}]
 upperarm_twist_02_r [{shapeType=Capsule, length=23.25, radius=6.75, rotation=<rot -85.9641 -179.9999 179.9999>}]
}
constraints [
 {bone1Name=dyn_key_mstr_neck, bone2Name=spine_05, swing1LimitDegrees=15, swing2LimitDegrees=15, twistLimitDegrees=5}
 {
  bone1Name=spine_05
  bone2Name=pelvis
  swing1LimitDegrees=45
  swing1Motion=Free
  swing2LimitDegrees=45
  swing2Motion=Free
  twistLimitDegrees=45
  twistMotion=Free
 }
 {bone1Name=dyn_key_neck_l, bone2Name=dyn_key_mstr_neck, swing1LimitDegrees=45, swing2LimitDegrees=45, twistLimitDegrees=45}
 {bone1Name=dyn_key_neck_r, bone2Name=dyn_key_mstr_neck, swing1LimitDegrees=45, swing2LimitDegrees=45, twistLimitDegrees=45}
 {bone1Name=dyn_key_neck_c, bone2Name=dyn_key_mstr_neck, swing1LimitDegrees=45, swing2LimitDegrees=45, twistLimitDegrees=45}
]
