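# AnimBP for the Master Key male medium body cosmetic, copying the attached parent mesh's pose and adding secondary motion.
# While targeting, it blends in upper-arm twist scalar bone scaling.
# RigidBody passes simulate flaps, body and neck physics assets, with weighted constraints on the two key bones before the neck pass.
# Front coat tails are then offset additively by thigh swing, remapped each update; a crouch-driven tail offset is left unconnected.
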AnimBlueprint /BRCosmetics/Characters/Player/Male/Medium/Bodies/M_MED_MasterKey/Meshes/M_MED_Master_Key_Body_AnimBP
target_skeleton=&/M_MED_Master_Key_Body_Skeleton
graphs [
 AnimGraph {nodes [
   "Transform (Modify) Bone - Bone: dyn_coat_tail_front_l"(Alpha="To Float (Boolean)"(InBool=Get bIsCrouching()), Translation="0,0,3", Rotation="0,0,0", Scale="1,1,1") {
    rotationMode=BMM_Additive
    rotationSpace=BCS_BoneSpace
    translationMode=BMM_Additive
    translationSpace=BCS_BoneSpace
    alphaBoolBlend {blendOption=Linear}
    boneToModify {boneName=dyn_coat_tail_front_l}
    translation {z=3}
   }
   n1 = "Transform (Modify) Bone - Bone: upperarm_twist_01_r_scalar"!(ComponentPose=Local To Component(LocalPose=Use cached pose 'CPFM_Cache'()), Alpha=1, Scale="0.65,0.65,0.65") -> (Pose) {
    scaleMode=BMM_Additive
    scaleSpace=BCS_BoneSpace
    alphaBoolBlend {blendOption=Linear}
    boneToModify {boneName=upperarm_twist_01_r_scalar}
    rotation {yaw=-60}
    scale=<0.65 0.65 0.65>
   }
   n2 = "Transform (Modify) Bone - Bone: upperarm_twist_01_l_scalar"!(ComponentPose=@n1, Alpha=1, Scale="0.65,0.65,0.65") -> (Pose) {
    scaleMode=BMM_Additive
    scaleSpace=BCS_BoneSpace
    alphaBoolBlend {blendOption=Linear}
    boneToModify {boneName=upperarm_twist_01_l_scalar}
    rotation {yaw=-60}
    scale=<0.65 0.65 0.65>
   }
   n3 = Component To Local(ComponentPose=@n2)
   n4 = Use cached pose 'CPFM_Cache'()
   n5 = Get bIsTargeting()
   n6 = RigidBody(ComponentPose=Local To Component(LocalPose=Blend Poses by bool(BlendPose_0=@n3, BlendPose_1=@n4, bActiveValue=@n5) {blendType=Linear})) {
    bOverrideWorldGravity=true
    overridePhysicsAsset=&/M_MED_Master_Key_Body_Flaps_Physics
    simulationSpace=BaseBoneSpace
    alphaBoolBlend {blendOption=Linear}
    baseBoneRef {boneName=pelvis}
    componentAppliedLinearAccClamp=<1000 1000 1000>
    componentLinearAccScale=<1 1 1>
    componentLinearVelScale=<1 1 1>
    overrideWorldGravity {z=-450}
   }
   n7 = RigidBody(ComponentPose=@n6, ExternalForce="0,0,1250") {
    bClampLinearTranslationLimitToRefPose=true
    bOverrideWorldGravity=true
    overridePhysicsAsset=&/M_MED_Master_Key_Body_Physics
    simulationSpace=BaseBoneSpace
    alphaBoolBlend {blendOption=Linear}
    alphaScaleBiasClamp {interpSpeedDecreasing=6, interpSpeedIncreasing=7}
    baseBoneRef {boneName=root}
    componentAppliedLinearAccClamp=<6000 6000 6000>
    componentLinearAccScale=<0.25 0.25 0.25>
    componentLinearVelScale=<0.25 0.25 0.25>
    overrideWorldGravity {z=-980}
   }
   n8 = "Constraint - dyn_key_mstr_l (PP)"!(ComponentPose=@n7, Alpha=1, ConstraintWeights_0=0.1237, ConstraintWeights_1:float=0.8763) -> (Pose) {
    alphaBoolBlend {blendOption=Linear}
    boneToModify {boneName=dyn_key_mstr_l}
    constraintSetup [{targetBone {boneName=spine_01}}, {targetBone {boneName=pelvis}}]
    constraintWeights [1,1]
   }
   n9 = "Constraint - dyn_key_mstr_r (PPP)"!(ComponentPose=@n8, Alpha=1, ConstraintWeights_0=0.1314, ConstraintWeights_1:float=0.0019, ConstraintWeights_2:float=0.8666) -> (Pose) {
    alphaBoolBlend {blendOption=Linear}
    boneToModify {boneName=dyn_key_mstr_r}
    constraintSetup [{targetBone {boneName=spine_01}}, {targetBone {boneName=thigh_twist_01_r}}, {targetBone {boneName=pelvis}}]
    constraintWeights [1,1,1]
   }
   n10 = RigidBody(ComponentPose=@n9) {
    bClampLinearTranslationLimitToRefPose=true
    bOverrideWorldGravity=true
    overridePhysicsAsset=&/M_MED_Master_Key_Neck_Physics
    simulationSpace=BaseBoneSpace
    alphaBoolBlend {blendOption=Linear}
    baseBoneRef {boneName=root}
    componentAppliedLinearAccClamp=<50 50 50>
    componentLinearAccScale=<0.3 0.3 0.3>
    componentLinearVelScale=<0.3 0.3 0.3>
    overrideWorldGravity {z=-980}
   }
   n11 = "Transform (Modify) Bone - Bone: dyn_coat_tail_front_l"(ComponentPose=@n10, Alpha=Get legSwingDelta_l() -> (legSwingDelta_l:double), Translation="-5,-5,10", Rotation="90,0,0", Scale="1,1,1") {
    rotationMode=BMM_Additive
    rotationSpace=BCS_BoneSpace
    translationMode=BMM_Additive
    translationSpace=BCS_BoneSpace
    alphaBoolBlend {blendOption=Linear}
    boneToModify {boneName=dyn_coat_tail_front_l}
    rotation {pitch=90}
    translation=<-5 -5 10>
   }
   n12 = "Transform (Modify) Bone - Bone: dyn_coat_tail_front_r"(ComponentPose=@n11, Alpha=Get legSwingDelta_r() -> (legSwingDelta_r:double), Translation="0,0,10", Rotation="80,0,10", Scale="1,1,1") {
    rotationMode=BMM_Additive
    rotationSpace=BCS_BoneSpace
    translationMode=BMM_Additive
    translationSpace=BCS_BoneSpace
    alphaBoolBlend {blendOption=Linear}
    alphaScaleBiasClamp {bInterpResult=true, interpSpeedDecreasing=8, interpSpeedIncreasing=9}
    boneToModify {boneName=dyn_coat_tail_front_r}
    rotation {pitch=80, roll=10}
    translation {z=10}
   }
   Output Pose(Result=Component To Local(ComponentPose=@n12)) {layerGroup=None}
   Save cached pose 'CPFM_Cache'(Pose=Copy Pose From Mesh() {bUseAttachedParent=true}) {cacheName=CPFM_Cache}
  ]}
 EventGraph {nodes [
   Event BlueprintUpdateAnimation()
   n1 = GetOwnerSkeletalMeshComponent()
   > Set legSwingDelta_r(legSwingDelta_r:double=MapRangeClamped(Value=GetTwistAndSwingAngleOfDeltaRotationFromRefPose(self=@n1, BoneName=thigh_r).OutSwingAngle, InRangeA=155, InRangeB=75)) -> (Output_Get:double)
   > Set legSwingDelta_l(legSwingDelta_l:double=MapRangeClamped(Value=GetTwistAndSwingAngleOfDeltaRotationFromRefPose(self=@n1, BoneName=thigh_l).OutSwingAngle, InRangeA=29, InRangeB=107)) -> (Output_Get:double)
   TryGetPawnOwner()
  ]}
]
variables [
 legSwingDelta_l {type=number}
 legSwingDelta_r {type=number}
]
