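# Physics asset for the hair face accessory of the male medium Master Key cosmetic character.
# Two zero-length capsules act as spheres: a small body on the dyn_tail bone and a larger offset one on the head bone.
# A single constraint joins the tail to the head, with narrow swing1 and twist limits and a zero swing2 limit.
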PhysicsAsset /BRCosmetics/Characters/Player/Male/Medium/Bodies/M_MED_MasterKey/Meshes/M_MED_Master_Key_FaceAcc_Hair_Physics
bodies {
 dyn_tail [{shapeType=Capsule, length=0, radius=3.1373, center=<0 -0.1 0>}]
 head [{shapeType=Capsule, length=0, radius=10.5644, center=<8.4201 2.9658 -0.2335>, rotation=<rot 0 -49.59 -83.618>}]
}
constraints [{bone1Name=dyn_tail, bone2Name=head, swing1LimitDegrees=12.5, swing2LimitDegrees=0, twistLimitDegrees=22.5}]
